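# Translucent unlit, two-sided material for the dragon part of the Masako Reign skydiving contrail, used on Niagara ribbons and mesh particles.
# Color layers tiled scale gradients, panning noise and two-color gradients, with tail falloff and a tinted leading end, scaled by particle color.
# Opacity combines UV masks, particle alpha, view-angle fades and optional height-lerp dissolve.
# Dynamic parameters drive dissolve and optional mesh-width and camera offsets.
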Material /BRCosmetics/Athena/Effects/Particles/Skydiving/NS_Contrail_MasakoReign/M_Contrail_MasakoReign_Dragon
graph {
 properties {
  bAutomaticallySetUsageInEditor=false
  bUsedWithNiagaraMeshParticles=true
  bUsedWithNiagaraRibbons=true
  blendMode=BLEND_Translucent
  shadingModel=MSM_Unlit
  twoSided=true
  usedShadingModels=Unlit
 }
 nodes [
  n1 = TextureCoordinate()
  n2 = Multiply(
   A=Subtract(A=ComponentMask(None=@n1) {g=true}) {constB=0.5}
   B=MaterialFunctionCall() {materialFunction=&TangentBasis}.Y V
  )
  n3 = DynamicParameter() {defaultValue=<rgba 1 1 0.5 0>, paramNames [Offset, Dissolve, Cameraoffset, CustomMeshWidth]}
  n4 = ScalarParameter() {parameterName=UseCustomMeshWidth}
  n5 = DynamicParameter() {defaultValue=<rgba 1 1 0.5 0>, paramNames [Offset, Dissolve, Cameraoffset, CustomMeshWidth]}
  Add(
   A=LinearInterpolate(B=Multiply(A=@n2, B=@n3.CustomMeshWidth), Alpha=@n4)
   B=LinearInterpolate(
    B=MaterialFunctionCall(Offset Amount=@n5.Cameraoffset) {materialFunction=/Engine/Functions/Engine_MaterialFunctions02/WorldPositionOffset/CameraOffset}.Camera Offset
    Alpha=ScalarParameter() {parameterName=UseCameraOffset}
   )
  )
  n6 = TextureCoordinate() {uTiling=2, vTiling=0.5}
  LinearInterpolate(Alpha=TextureSample(UVs=Panner(Coordinate=@n6) {speedX=-2}) {samplerType=SAMPLERTYPE_Grayscale, texture=&LowResBlurredNoise}.R) {constA=0.7}
  n7 = ComponentMask(None=TextureCoordinate()) {g=true}
  n8 = Multiply(A=Power(Base=Sine(None=@n7) {period=2}) {constExponent=5}) {constB=20}
  n9 = TextureCoordinate()
  n10 = MaterialFunctionCall(In=Add(A=ComponentMask(None=@n9) {g=true}) {constB=0.3}, Contrast=ScalarParameter() {defaultValue=3, parameterName="Contrast (S)_2"}) {materialFunction=&CheapContrast}
  n11 = Multiply(A=ParticleColor().A, B=Multiply(A=Saturate(None=@n8), B=Multiply(A=@n10.Result)))
  n12 = MaterialFunctionCall("FadeSensitivity (optional)"=ScalarParameter() {defaultValue=0.5, parameterName=AngleFade}) {materialFunction=&FadeBasedOnViewAngle}
  n13 = ScalarParameter() {defaultValue=0.5, parameterName="Contrast (S)"}
  n14 = TextureCoordinate()
  n15 = Power(Base=OneMinus(None=ComponentMask(None=@n14) {r=true})) {constExponent=1.5}
  n16 = Multiply(
   A=Multiply(
    A=Multiply(A=@n11)
    B=MaterialFunctionCall(In=@n12.Result, Contrast=@n13) {materialFunction=&CheapContrast}.Result
   )
   B=Saturate(None=Multiply(A=@n15) {constB=20})
  )
  n17 = LinearInterpolate(
   B=MaterialFunctionCall("FadeSensitivity (optional)"=ScalarParameter() {defaultValue=1, parameterName=AngleFadeMesh}) {materialFunction=&FadeBasedOnViewAngle}.Result
   Alpha=@n4
  ) {constA=1}
  n18 = MaterialFunctionCall(
   Transition Phase=DynamicParameter() {defaultValue=<rgba 1 1 0.5 0>, paramNames [Offset, Dissolve, Cameraoffset, CustomMeshWidth]}.Dissolve
   Height Texture=TextureSample(UVs=TextureCoordinate() {vTiling=0.5}) {samplerType=SAMPLERTYPE_LinearColor, texture=/Game/Effects/Fort_Effects/BluGlow/Violentium/Violentium_Mask}.R
   Contrast=Constant() {r=1}
  ) {materialFunction=&HeightLerp}
  n19 = ScalarParameter() {parameterName=UseDissolve}
  Reroute(None=Multiply(A=Multiply(A=@n16, B=@n17), B=LinearInterpolate(B=@n18.Results, Alpha=@n19) {constA=1}))
  n20 = ParticleColor()
  n21 = Multiply(
   A=TextureCoordinate()
   B=AppendVector(A=ScalarParameter() {defaultValue=5, parameterName=BaseTiling_U}, B=ScalarParameter() {defaultValue=0.55, parameterName=BaseTiling_V})
  )
  n22 = Multiply(
   A=VectorParameter() {parameterName=ScalesColor, defaultValue=<rgba 1.5 1.1719 0.7969 1>}.RGB
   B=TextureSample(UVs=@n21) {texture=&/T_MasakoReign_Scales_Gradient_Tile}.R
  )
  n23 = Multiply(
   A=TextureSample(UVs=@n21) {texture=&T_MasakoReign_Scales_Gradient_Tile}.G
   B=VectorParameter() {defaultValue=<rgba 3 2.3286 1.2741 1>}.RGB
  )
  n24 = Sine(None=ComponentMask(None=TextureCoordinate()) {g=true}) {period=2}
  n25 = Multiply(A=Constant() {r=-1.5}, B=ScalarParameter() {defaultValue=1, parameterName=NoiseSpeed})
  n26 = Constant()
  n27 = TextureSample(UVs=Panner(Coordinate=TextureCoordinate() {uTiling=2, vTiling=0.5}, Speed=AppendVector(A=@n25, B=@n26)) {speedX=-1.5}) {samplerType=SAMPLERTYPE_Grayscale, texture=&LowResBlurredNoise}
  n28 = Multiply(
   A=Multiply(A=Add(A=@n22, B=@n23), B=Power(Base=@n24) {constExponent=1})
   B=LinearInterpolate(Alpha=Power(Base=@n27.R) {constExponent=1.1}) {constB=2}
  )
  n29 = Power(Base=SphereMask(A=TextureCoordinate() {uTiling=0}, B=Constant2Vector() {g=0.5}) {attenuationRadius=0.5, hardnessPercent=0}) {constExponent=1.2}
  n30 = VectorParameter() {parameterName=Color2, defaultValue=<rgba 0.2 0 0 0>}
  n31 = VectorParameter() {parameterName=Color1, defaultValue=<rgba 1 0 0 1>}
  n32 = SphereMask(A=TextureCoordinate() {uTiling=0}, B=Constant2Vector() {g=0.5}) {attenuationRadius=0.5, hardnessPercent=0}
  n33 = Multiply(A=MaterialFunctionCall(In=Power(Base=@n32) {constExponent=1.5}, Contrast=Constant() {r=0.1}) {materialFunction=&CheapContrast}.Result)
  n34 = ScalarParameter() {defaultValue=0.3, parameterName=NoiseLerpMin}
  n35 = TextureCoordinate() {vTiling=0.3}
  n36 = ComponentMask(None=VectorParameter() {parameterName=NoiseUVScale, defaultValue=<rgba 1 1 0 0>}.RGB) {g=true, r=true}
  n37 = Multiply(A=Constant() {r=-0.7}, B=ScalarParameter() {defaultValue=1, parameterName=NoiseSpeed})
  n38 = Constant()
  n39 = TextureSample(UVs=Panner(Coordinate=Multiply(A=@n35, B=@n36), Speed=AppendVector(A=@n37, B=@n38)) {speedX=-0.7}) {samplerType=SAMPLERTYPE_Grayscale, texture=/BRCosmetics/Weapons/FORT_Melee/Pickaxe_Cube_Ninja_Male/FX/T_CellNoise}
  n40 = Multiply(
   A=Add(A=Multiply(A=@n28, B=@n29), B=LinearInterpolate(A=@n30.RGB, B=@n31.RGB, Alpha=@n33))
   B=LinearInterpolate(B=LinearInterpolate(A=@n34, Alpha=@n39.R) {constA=0.3, constB=1.5}) {constAlpha=1}
  )
  n41 = Power(Base=ComponentMask(None=TextureCoordinate()) {g=true}) {constExponent=2.5}
  n42 = ScalarParameter() {defaultValue=0.1, parameterName="Contrast (S)_1"}
  n43 = LinearInterpolate(Alpha=Multiply(A=MaterialFunctionCall(In=@n41, Contrast=@n42) {materialFunction=&CheapContrast}.Result)) {constA=0.3, constB=1.5}
  n44 = ScalarParameter() {defaultValue=0.6, parameterName=TailGradient}
  n45 = TextureCoordinate()
  n46 = Power(Base=OneMinus(None=ComponentMask(None=@n45) {r=true}))
  n47 = ComponentMask(None=TextureCoordinate()) {r=true}
  n48 = Multiply(A=Power(Base=OneMinus(None=@n47), Exp=ScalarParameter() {defaultValue=20, parameterName=LeadPower}) {constExponent=20}) {constB=10}
  n49 = Multiply(
   A=Multiply(A=Multiply(A=@n40, B=@n43), B=LinearInterpolate(A=@n44, Alpha=@n46) {constA=0.6, constB=2})
   B=LinearInterpolate(B=VectorParameter() {parameterName=Param_1, defaultValue=<rgba 1.7372 2 0.4229 1>}.RGB, Alpha=Saturate(None=@n48)) {constA=1, constB=3}
  )
  Reroute(None=MaterialFunctionCall(Emissive=Multiply(A=@n20.RGB, B=@n49)) {materialFunction=&MF_FX_EmissiveAdjustments}.Result)
 ]
}
