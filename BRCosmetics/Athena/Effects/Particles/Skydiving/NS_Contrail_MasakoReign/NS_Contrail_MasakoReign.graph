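# Niagara skydiving contrail for the MasakoReign cosmetic, drawing a pair of fiery dragon trails from skeletal mesh bones or sockets.
# Each trail combines a lit head mesh and a scratch-module-animated body ribbon, bobbed by offset waveforms.
# Mesh flames, sprites and smoke add fire around them.
# User.ParaGlideLeanAlpha scales spawn probability, ribbon offset and sprite alpha, while front-end and platform switches adjust size and brightness.
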NiagaraSystem /BRCosmetics/Athena/Effects/Particles/Skydiving/NS_Contrail_MasakoReign/NS_Contrail_MasakoReign
dependencies {
 usedDataInterfaces [
  &NiagaraDataInterfaceArrayFloat3
  &NiagaraDataInterfaceArrayFloat
  &NiagaraDataInterfaceColorCurve
  &NiagaraDataInterfaceCurve
  &NiagaraDataInterfacePlatformSet
  &NiagaraDataInterfaceArrayInt32
  &NiagaraDataInterfaceSkeletalMesh
 ]
 usedDynamicInputs []
 usedModules [
  &self:SystemSpawnScript.SystemScriptSource.SystemScriptGraph.NiagaraNodeAssignment_0.SetVariables_#1
  &self:SystemSpawnScript.SystemScriptSource.SystemScriptGraph.NiagaraNodeAssignment_1.SetVariables_#2
  &self:SystemSpawnScript.SystemScriptSource.SystemScriptGraph.NiagaraNodeAssignment_4.SetVariables_#3
  &SystemState
  &self:SystemSpawnScript.SystemScriptSource.SystemScriptGraph.NiagaraNodeAssignment_3.SetVariables_#4
 ]
}
emitters {
 CurveFire {
  emitterUpdateScript [
   &EmitterState(Life Cycle Mode=::NewEnumerator1, Loop Duration=0.1, UseLoopDelay=false)
   &SpawnRate(Spawn Probability=&Lerp_Float(A=0, B=1, Alpha=$User.ParaGlideLeanAlpha), SpawnRate=8, Use Spawn Probability=true)
   &SpawnBurst_Instantaneous(
    Spawn Probability=&Lerp_Float(A=0.6, B=1, Alpha=$User.ParaGlideLeanAlpha)
    Use Spawn Probability=true
    UseLoopCountLimit=false
   )
   /SetVariables(Emitter.Allsize=0.8) {isSetParameters=true}
  ]
  particleSpawnScript [
   &PartitionParticles()
   &InitializeParticle(
    Color=<rgba 0.6 0.1466 0.1389 5>
    Lifetime Max=1
    Lifetime Mode=::NewEnumerator1
    Mesh Scale=<0.5 0.5 0.5>
    Mesh Scale Mode=::NewEnumerator4
    Position Offset=&RandomRangeVector(Minimum=<0 0 -25>, Maximum=<0 0 20>)
    UsePositionOffset=true
    Write Material Random=true
   )
   &SkeletalMeshLocation(
    Bone / Socket Index=$Particles.PartitionParticles.Partition
    Bone Sampling Mode=::NewEnumerator4
    MaskSpawnGroup=false
    Orientation Sampling=::NewEnumerator1
    Sampled Position Offset=&AddVectorToPosition(
     Position=&SelectVectorFromArray(
      Direct Array Index=$Particles.PartitionParticles.Partition
      Vector Selection Array=DataInterface() {propertyValues {FloatData [<40 0 0>, <-40 0 0>]}}
     )
     Vector=<0 0 30>
    )
    Sampled Position Offset Coordinate Space:ENiagaraSkelMeshTransforms=::NewEnumerator2
    Skeletal Mesh=$System.Skeletal Mesh Reader
   )
   &ShapeLocation(MaskSpawnGroup=false, Override Seed=false, Sphere Radius=1, UseOwnerScale=true)
   &InitialMeshOrientation(Mesh Orientation Mode=::NewEnumerator7, Rotation=<0 0 0.25>, Use Rotation Vector=true)
   &InitialMeshOrientation(
    Mesh Orientation Mode=::NewEnumerator2
    Rotation=&RandomRangeVector(Minimum=<-1 0 0>, Maximum=<1 0 0>)
    Use Rotation Vector=true
   ) {module=InitialMeshOrientation001}
   &AddVelocity(
    Cone Angle=0
    Cone Axis=<0 0 1>
    Module.UseConeFalloff=false
    Module.UseVelDistribution=false
    Override Seed=false
    Velocity Mode=::NewEnumerator2
    Velocity Speed=&RandomRangeFloat(Minimum=200, Maximum=200)
   )
  ]
  particleUpdateScript [
   /SetVariables(Particles.AlphaAll:NiagaraFloat=&FloatFromCurve(
     FloatCurve=DataInterface() {propertyValues {Curve {defaultValue=1, keys [{tangentMode=RCTM_User, time=0, value=0}, {tangentMode=RCTM_User, time=0.5, value=1}]}}}
     Scale Curve=&FloatFromCurve(
      FloatCurve=DataInterface() {propertyValues {Curve {keys [{arriveTangent=-2.8607, interpMode=RCIM_Cubic, leaveTangent=-2.8607, tangentMode=RCTM_User, time=0.0044, value=0.9529}, <at 1 0>]}}}
      Scale Curve=&Multiply_Float(A=1.5, B=1)
     )
    )) {isSetParameters=true}
   /SetVariables(Particles.BrightnessAll:NiagaraFloat=&FloatFromCurve(
     FloatCurve=DataInterface() {propertyValues {Curve {keys [
         <at 0 -0.0118 RCIM_Cubic>
         {arriveTangent=0.3618, interpMode=RCIM_Cubic, leaveTangent=0.3618, time=0.2035, value=1}
         {arriveTangent=-1.2554, interpMode=RCIM_Cubic, leaveTangent=-1.2554, time=0.5284, value=0.1794}
         <at 1 0>
        ]}}}
     Scale Curve=&MakeCustomFloatFromBool(Bool=&RandomBool(), True Float=0.3, False Float=1)
    )) {isSetParameters=true}
   &DynamicMaterialParameters(
    Index 0 Param 1=&Multiply_Float(A=1.1, B=$Emitter.Allsize)
    Index 0 Param 2=2
    Index 0 Param 3=&RandomRangeFloat(Minimum=0, Maximum=1)
    Index 0 Param 4=0.5
    Index 1 Param 1=&FloatFromCurve(FloatCurve=DataInterface() {propertyValues {Curve {keys [<at 0 0>,<at 1 1>]}}})
    Index 1 Param 2=0.3
    Module.Index0Float0Write=true
    Module.Index0Float1Write=true
    Module.Index0Float2Write=true
    Module.Index0Float3Write=true
    Module.Index1Float0Write=true
    Module.Index1Float1Write=true
    Module.Index1Float2Write=false
    Module.Index1Float3Write=false
    Write Parameter Index 1=::NewEnumerator1
   )
   &ParticleState()
   &UpdateMeshOrientation(Rotation Rate=-0.2, Rotation Vector=<0 1 0>)
   &ScaleMeshSize(Scale Factor=&Multiply_VectorByFloat(
     Vector=&MakeCustomVectorFromBool(Bool=&RandomBool(), True Vector=<-1 1 1>, False Vector=<1 1 1>)
     Float=&FloatFromCurve(
      FloatCurve=DataInterface() {propertyValues {Curve {
         defaultValue=1
         keys [{arriveTangent=1.0428, interpMode=RCIM_Cubic, time=0, value=1.2}, {interpMode=RCIM_Cubic, leaveTangent=0.8255, time=1, value=1}]
        }}}
      Scale Curve=&Multiply_Float(A=&RandomRangeFloat(Minimum=0.7, Maximum=0.8), B=$Emitter.Allsize)
     )
    ))
   &Drag(Use Linear Drag=true, Use Rotational Drag=true)
   &SolveForcesAndVelocity(Clamp Velocity=false, Limit Acceleration=false)
   &ScaleColor(
    Scale Alpha=$Particles.AlphaAll
    Scale RGB=&Multiply_VectorByFloat(
     Vector=&MakeVectorFromLinearColorRGB(Color=&ColorFromCurve(ColorCurve=$System.Curve for Colors))
     Float=$Particles.BrightnessAll
    )
    ScaleA=true
    ScaleRGB=true
   )
   &FN_Mesh_Attributes(
    Color=&MakeLinearColorFromVectorAndFloat(
     "Vector (RGB)"=&Multiply_Vector(
      A=&Multiply_VectorByFloat(Vector=&MakeVectorFromLinearColorRGB(Color=<rgba 5 0.4167 0 0.3>), Float=$Particles.BrightnessAll)
      B=&MakeVectorFromLinearColorRGB(Color=&ColorFromCurve(ColorCurve=$System.Curve for Colors))
     )
     "Float (Alpha)"=&Multiply_Float(A=0.3, B=$Particles.AlphaAll)
    )
    Create Color Attribute=true
    Module Output=::NewEnumerator1
   )
  ]
  rendererClasses [&NiagaraMeshRendererProperties]
  renderers [
   {rendererClass=&NiagaraMeshRendererProperties, rendererIndex=0}
   {rendererClass=&NiagaraMeshRendererProperties, rendererIndex=1}
  ]
 }
 Dragon {
  emitterUpdateScript [
   &EmitterState(Life Cycle Mode=::NewEnumerator1, Loop Behavior=::NewEnumerator1, Loop Duration Mode=::NewEnumerator1, UseLoopDelay=false)
   &SpawnBurst_Instantaneous(Spawn Count=60, Use Spawn Probability=false, UseLoopCountLimit=false)
  ]
  particleSpawnScript [
   &PartitionParticles()
   &InitializeParticle(Color Mode=::NewEnumerator0, Ribbon Width=43, Ribbon Width Mode=::NewEnumerator1, UsePositionOffset=false, Write Material Random=true)
  ]
  particleUpdateScript [
   &ParticleState(Kill Particles When Lifetime Has Elapsed=false)
   &SkeletalMeshLocation(
    Bone / Socket Index=$Particles.PartitionParticles.Partition
    Bone Sampling Mode=::NewEnumerator4
    MaskSpawnGroup=false
    Orientation Sampling=::NewEnumerator1
    Sampled Position Offset=&AddVectorToPosition(
     Position=&AddVectorToPosition(
      Position=&SelectVectorFromArray(
       Direct Array Index=$Particles.PartitionParticles.Partition
       Vector Selection Array=DataInterface() {propertyValues {FloatData [<40 0 0>, <-40 0 0>]}}
      )
      Vector=&Lerp_Vector(
       A=&Lerp_Vector(A=<0 0 300>, B=<0 0 400>, Alpha=$User.ParaGlideLeanAlpha)
       B=<0 0 0>
       Alpha=&ReturnNormalizedExecIndex()
      )
     )
     Vector=&Lerp_Vector(
      A=<0 0 0>
      B=<0 0 20>
      Alpha=&Waveform(
       "Waveform [1]"=::NewEnumerator0
       "[1] Phase Scale"=&SelectFloatFromArray(Direct Array Index=$Particles.PartitionParticles.Partition, Float Selection Array=DataInterface() {propertyValues {FloatData [0.5,0.7]}})
       "[1] Frequency"=&SelectFloatFromArray(Direct Array Index=$Particles.PartitionParticles.Partition, Float Selection Array=DataInterface() {propertyValues {FloatData [0.5,0.7]}})
       "[1] Amplitude Min/Max"=<0 1>
       "[1] Scale"=1.5
      )
     )
    )
    Sampled Position Offset Coordinate Space:ENiagaraSkelMeshTransforms=::NewEnumerator2
    Skeletal Mesh=$System.Skeletal Mesh Reader
   )
   &self:ScratchModule(
    Age:NiagaraFloat=$Emitter.Age
    Curve for Colors=DataInterface() {propertyValues {
      AlphaCurve {keys [<at 0 1>,<at 1 1>]}
      BlueCurve {keys [
        {interpMode=RCIM_Cubic, leaveTangentWeight=0.0518, tangentMode=RCTM_User, tangentWeightMode=RCTWM_WeightedBoth, time=0.0031, value=0}
        {
         arriveTangent=0.0052
         arriveTangentWeight=0.1917
         interpMode=RCIM_Cubic
         leaveTangent=0.0052
         leaveTangentWeight=0.1917
         tangentMode=RCTM_User
         tangentWeightMode=RCTWM_WeightedBoth
         time=0.1371
         value=0
        }
        {
         arriveTangent=0.0293
         arriveTangentWeight=0.0977
         interpMode=RCIM_Cubic
         leaveTangent=0.0293
         leaveTangentWeight=0.0977
         tangentMode=RCTM_User
         tangentWeightMode=RCTWM_WeightedBoth
         time=0.3191
         value=0
        }
        {
         arriveTangentWeight=0.1033
         interpMode=RCIM_Cubic
         leaveTangentWeight=0.1033
         tangentMode=RCTM_User
         tangentWeightMode=RCTWM_WeightedBoth
         time=0.5022
         value=0
        }
        {
         arriveTangentWeight=0.1042
         interpMode=RCIM_Cubic
         leaveTangentWeight=0.1042
         tangentMode=RCTM_User
         tangentWeightMode=RCTWM_WeightedBoth
         time=0.6685
         value=0
        }
        {
         arriveTangentWeight=0.0897
         interpMode=RCIM_Cubic
         leaveTangentWeight=0.0897
         tangentMode=RCTM_User
         tangentWeightMode=RCTWM_WeightedBoth
         time=0.8591
         value=0
        }
        {arriveTangentWeight=0.0561, interpMode=RCIM_Cubic, tangentMode=RCTM_User, tangentWeightMode=RCTWM_WeightedBoth, time=1, value=0}
       ]}
      GreenCurve {keys [
        <at 0 0 RCIM_Cubic>
        {
         arriveTangent=-0.0991
         arriveTangentWeight=0.0819
         interpMode=RCIM_Cubic
         leaveTangent=-0.0991
         leaveTangentWeight=0.0819
         tangentMode=RCTM_User
         tangentWeightMode=RCTWM_WeightedBoth
         time=0.1474
         value=0.1912
        }
        {
         arriveTangent=-0.0109
         arriveTangentWeight=0.1035
         interpMode=RCIM_Cubic
         leaveTangent=-0.0109
         leaveTangentWeight=0.1035
         tangentMode=RCTM_User
         tangentWeightMode=RCTWM_WeightedBoth
         time=0.3256
         value=0
        }
        {
         arriveTangent=-0.0362
         arriveTangentWeight=0.088
         interpMode=RCIM_Cubic
         leaveTangent=-0.0362
         leaveTangentWeight=0.088
         tangentMode=RCTM_User
         tangentWeightMode=RCTWM_WeightedBoth
         time=0.5156
         value=0.5563
        }
        {arriveTangent=-1.1282, interpMode=RCIM_Cubic, leaveTangent=-1.1282, time=0.7015, value=0}
        {
         arriveTangentWeight=0.0831
         interpMode=RCIM_Cubic
         leaveTangentWeight=0.0831
         tangentMode=RCTM_User
         tangentWeightMode=RCTWM_WeightedBoth
         time=0.8534
         value=0.1752
        }
        <at 1 0>
       ]}
      RedCurve {keys [
        {arriveTangent=2.3812, interpMode=RCIM_Cubic, leaveTangent=2.3812, tangentMode=RCTM_User, time=0, value=0}
        {
         arriveTangent=0.0019
         arriveTangentWeight=0.0857
         interpMode=RCIM_Cubic
         leaveTangent=0.0019
         leaveTangentWeight=0.0857
         tangentMode=RCTM_User
         tangentWeightMode=RCTWM_WeightedBoth
         time=0.0007
         value=0.1996
        }
        {
         arriveTangent=-0.0076
         arriveTangentWeight=0.0935
         interpMode=RCIM_Cubic
         leaveTangent=-0.0076
         leaveTangentWeight=0.0935
         tangentMode=RCTM_User
         tangentWeightMode=RCTWM_WeightedBoth
         time=0.1528
         value=-0.1971
        }
        {
         arriveTangent=0.0038
         arriveTangentWeight=0.1046
         interpMode=RCIM_Cubic
         leaveTangent=0.0038
         leaveTangentWeight=0.1046
         tangentMode=RCTM_User
         tangentWeightMode=RCTWM_WeightedBoth
         time=0.3011
         value=0.6914
        }
        {
         arriveTangent=0.1
         arriveTangentWeight=0.133
         interpMode=RCIM_Cubic
         leaveTangent=0.1
         leaveTangentWeight=0.133
         tangentMode=RCTM_User
         tangentWeightMode=RCTWM_WeightedBoth
         time=0.5
         value=-0.3998
        }
        {
         arriveTangent=0.0115
         arriveTangentWeight=0.1019
         interpMode=RCIM_Cubic
         leaveTangent=0.0115
         leaveTangentWeight=0.1019
         tangentMode=RCTM_User
         tangentWeightMode=RCTWM_WeightedBoth
         time=0.695
         value=0.4189
        }
        {
         arriveTangent=0.011
         arriveTangentWeight=0.1017
         interpMode=RCIM_Cubic
         leaveTangent=0.011
         leaveTangentWeight=0.1017
         tangentMode=RCTM_User
         tangentWeightMode=RCTWM_WeightedBoth
         time=0.8289
         value=-0.124
        }
        {
         arriveTangentWeight=0.0783
         interpMode=RCIM_Cubic
         leaveTangentWeight=0.0783
         tangentMode=RCTM_User
         tangentWeightMode=RCTWM_WeightedBoth
         time=0.9985
         value=0.2005
        }
       ]}
     }}
    NormalizeINdex:NiagaraFloat=&ReturnNormalizedExecIndex()
    Offset:NiagaraFloat=&SelectFloatFromArray(Direct Array Index=$Particles.PartitionParticles.Partition, Float Selection Array=DataInterface() {propertyValues {FloatData [0,0.35]}})
    STR:NiagaraFloat=&Multiply_Float(
     A=&Multiply_Float(A=50, B=&Lerp_Float(A=0.6, B=1.2, Alpha=$User.ParaGlideLeanAlpha))
     B=&FloatFromCurve(
      FloatCurve=DataInterface() {propertyValues {Curve {keys [<at 0.0033 1.0235 RCIM_Cubic>, {arriveTangent=-1.0269, interpMode=RCIM_Cubic, leaveTangent=-1.0269, time=0.8138, value=0.6906}, <at 1 0>]}}}
      CurveIndex=&ReturnNormalizedExecIndex()
     )
    )
    Speed=1
   ) {module=ScratchModule}
   &ScaleRibbonWidth(Ribbon Width Scale=&FloatFromCurve(
     FloatCurve=DataInterface() {propertyValues {Curve {keys [
         {arriveTangent=1.7318, interpMode=RCIM_Cubic, leaveTangent=1.7318, tangentMode=RCTM_User, time=0, value=0.3}
         {arriveTangent=0.6, interpMode=RCIM_Cubic, leaveTangent=0.6, time=0.5412, value=0.7794}
         <at 1 0.9>
        ]}}}
     CurveIndex=&ReturnNormalizedExecIndex()
     Scale Curve=&Multiply_Float(A=1, B=$System.FEsizeMatch)
    ))
   &ScaleColor(Scale RGB=&Multiply_VectorByFloat(Vector=<1 1 1>, Float=1.5), ScaleA=true, ScaleRGB=true)
   /SetVariables(Particles.RibbonID=/Niagara/DynamicInputs/NiagaraID/MakeNiagaraIDFromInt(Index:NiagaraInt32=$Particles.PartitionParticles.Partition, Acquire Tag:NiagaraInt32=$Particles.PartitionParticles.Partition)) {isSetParameters=true}
  ]
  rendererClasses [&NiagaraRibbonRendererProperties]
  renderers [{rendererClass=&NiagaraRibbonRendererProperties}]
 }
 FireCylinder {
  emitterUpdateScript [
   &EmitterState(Life Cycle Mode=::NewEnumerator1, Loop Duration=0.5, UseLoopDelay=false)
   &SpawnRate(SpawnRate=20, Use Spawn Probability=false)
  ]
  particleSpawnScript [
   &PartitionParticles()
   &InitializeParticle(
    Color=&RandomRangeLinearColor(Minimum=<rgba 1 0 0.0903 1>, Maximum=<rgba 5 1.1375 0 1>)
    Mesh Scale=<1 0.8 0.8>
    Mesh Scale Mode=::NewEnumerator4
    Position Offset=<0 -2.5 0>
    UsePositionOffset=true
    Write Material Random=true
   )
   &InitialMeshOrientation(Mesh Orientation Mode=::NewEnumerator2, Rotation=<0 -0.5 0>, Use Rotation Vector=true)
   &InitialMeshOrientation(Mesh Orientation Mode=::NewEnumerator2, Rotation=<0 0.25 0>, Use Rotation Vector=true) {module=InitialMeshOrientation001}
   &CameraOffset(Camera Offset Amount=-5)
   &InitialMeshOrientation(
    Mesh Orientation Mode=::NewEnumerator2
    Rotation=&RandomRangeVector(Minimum=<0 0 0>, Maximum=<0 0 1>)
    Use Rotation Vector=true
   ) {module=InitialMeshOrientation002}
  ]
  particleUpdateScript [
   &ParticleState()
   &ScaleMeshSize(Scale Factor=&Multiply_VectorByFloat(
     Vector=<0.4 1 1>
     Float=&FloatFromCurve(
      FloatCurve=DataInterface() {propertyValues {Curve {keys [<at 0 0>,<at 1 0.8>]}}}
      Scale Curve=&Multiply_Float(A=&RandomRangeFloat(Minimum=0.7, Maximum=1), B=$System.FEsizeMatch)
     )
    ))
   &ScaleColor(
    Scale Alpha=&FloatFromCurve(
     FloatCurve=DataInterface() {propertyValues {Curve {
        defaultValue=1
        keys [
         {interpMode=RCIM_Cubic, tangentMode=RCTM_User, time=0, value=0}
         {interpMode=RCIM_Cubic, tangentMode=RCTM_User, time=0.5, value=1}
         {tangentMode=RCTM_User, time=1, value=0}
        ]
       }}}
     Scale Curve=2
    )
    Scale RGB=&Multiply_VectorByFloat(
     Vector=<1 1 1>
     Float=&FN_PickFloatDependOnPlatform(TargetPlatformActive=1, TargetPlatformNoActive=0.2)
    )
    ScaleA=true
    ScaleRGB=true
   )
   &DynamicMaterialParameters(
    Index 0 Param 1=&FloatFromCurve(FloatCurve=DataInterface() {propertyValues {Curve {keys [<at 0 0>,<at 1 0.5>]}}})
    Module.Index0Param0Write=true
    Module.Index0Param1Write=true
    Module.Index0Param2Write=true
    Module.Index0Param3Write=true
   )
   &SkeletalMeshLocation(
    Bone / Socket Index=$Particles.PartitionParticles.Partition
    Bone Sampling Mode=::NewEnumerator4
    MaskSpawnGroup=false
    Orientation Sampling=::NewEnumerator1
    Sampled Position Offset=&AddVectorToPosition(
     Position=&SelectVectorFromArray(
      Direct Array Index=$Particles.PartitionParticles.Partition
      Vector Selection Array=DataInterface() {propertyValues {FloatData [<40 0 -10>, <-40 0 -10>]}}
     )
     Vector=&Lerp_Vector(
      A=<0 0 0>
      B=<0 0 20>
      Alpha=&Waveform(
       "Waveform [1]"=::NewEnumerator0
       "[1] Phase Scale"=&SelectFloatFromArray(Direct Array Index=$Particles.PartitionParticles.Partition, Float Selection Array=DataInterface() {propertyValues {FloatData [0.5,0.7]}})
       "[1] Frequency"=&SelectFloatFromArray(Direct Array Index=$Particles.PartitionParticles.Partition, Float Selection Array=DataInterface() {propertyValues {FloatData [0.5,0.7]}})
       "[1] Amplitude Min/Max"=<0 1>
       "[1] Scale"=1.5
      )
     )
    )
    Sampled Position Offset Coordinate Space:ENiagaraSkelMeshTransforms=::NewEnumerator2
    Skeletal Mesh=$System.Skeletal Mesh Reader
   )
  ]
  rendererClasses [&NiagaraMeshRendererProperties]
  renderers [{rendererClass=&NiagaraMeshRendererProperties, rendererIndex=0}]
 }
 FireSprite {
  emitterUpdateScript [
   &EmitterState(Life Cycle Mode=::NewEnumerator1, Loop Duration=1, UseLoopDelay=false)
   &SpawnRate(SpawnRate=3, Use Spawn Probability=false)
   &SpawnBurst_Instantaneous(Spawn Count=4, Use Spawn Probability=false, UseLoopCountLimit=false) {enabled=false}
  ]
  particleSpawnScript [
   &PartitionParticles(Number Of Partitions=4)
   &InitializeParticle(
    Color=<rgba 0.2969 0 0 2>
    Lifetime Max=1.5
    Lifetime Min=0.5
    Lifetime Mode=::NewEnumerator1
    Mesh Scale=&Multiply_VectorByFloat(Vector=<1 1 1>, Float=0.7)
    Mesh Scale Mode=::NewEnumerator4
    Sprite Rotation Angle=-0.25
    Sprite Rotation Mode=::NewEnumerator1
    Sprite Size Mode=::NewEnumerator3
    Sprite UV Mode=::NewEnumerator2
    Uniform Sprite Size=30
    UsePositionOffset=false
    Write Material Random=true
   )
   &AddVelocity(Velocity=&RandomRangeVector(Minimum=<0 0 200>, Maximum=<0 0 300>))
  ]
  particleUpdateScript [
   /SetVariables(Particles.AlphaAll:NiagaraFloat=&FloatFromCurve(FloatCurve=DataInterface() {propertyValues {Curve {keys [<at 0 0>,<at 0.2014 0.9676>,<at 0.6059 1.0118>,<at 0.7758 0.4971>,<at 1 0>]}}})) {isSetParameters=true}
   /SetVariables(Particles.SizeAll:NiagaraFloat=&Multiply_Float(
     A=&FloatFromCurve(FloatCurve=DataInterface() {propertyValues {Curve {keys [<at 0 0>,<at 0.1 1.0265>,<at 0.7 1>,<at 1 0>]}}})
     B=1
    )) {isSetParameters=true}
   &ParticleState()
   &ScaleSpriteSize(Scale Sprite Size Mode=::NewEnumerator0, Uniform Scale Factor=$Particles.SizeAll)
   &SkeletalMeshLocation(
    Bone / Socket Index=&SelectIntFromArray(Direct Array Index=$Particles.PartitionParticles.Partition, Int Selection Array=DataInterface() {propertyValues {IntData [0,1,0,1]}})
    Bone Sampling Mode=::NewEnumerator4
    MaskSpawnGroup=false
    Orientation Sampling=::NewEnumerator1
    Position Sampling=::NewEnumerator1
    Sampled Position Offset=&AddVectorToPosition(
     Position=&AddVectorToPosition(
      Position=&SelectVectorFromArray(
       Direct Array Index=$Particles.PartitionParticles.Partition
       Vector Selection Array=DataInterface() {propertyValues {FloatData [<80 0 50>, <-80 0 80>, <-10 0 100>, <10 0 120>]}}
      )
      Vector=&RandomRangeVector(Minimum=<0 0 -30>, Maximum=<0 0 -100>)
     )
     Vector=<0 0 0>
    )
    Sampled Position Offset Coordinate Space:ENiagaraSkelMeshTransforms=::NewEnumerator2
    Skeletal Mesh=$System.Skeletal Mesh Reader
   )
   &DynamicMaterialParameters(
    Index 0 Param 3=$Engine.Time
    Module.Index0Float0Write=true
    Module.Index0Float1Write=true
    Module.Index0Float2Write=true
    Module.Index0Float3Write=true
   )
   &FN_Sprite_Attributes(
    Color=&MakeLinearColorFromVectorAndFloat(
     "Vector (RGB)"=&MakeVectorFromLinearColorRGB(Color=<rgba 20 0.3333 0 1>)
     "Float (Alpha)"=&Multiply_Float(A=&Lerp_Float(A=0, B=1.5, Alpha=$User.ParaGlideLeanAlpha), B=$Particles.AlphaAll)
    )
    Create Color Attribute=true
    Create SpriteSize Attribute=true
    Module Output:ENiagara_AttributeSamplingApplyOutput=::NewEnumerator1
    SpriteSize=&Multiply_Vector2D_ByFloat(Float=&Multiply_Float(A=31, B=$Particles.SizeAll), Vector2D=<1 1>)
   )
   &FN_Sprite_Attributes(
    Color=&MakeLinearColorFromVectorAndFloat(
     "Vector (RGB)"=&MakeVectorFromLinearColorRGB(Color=<rgba 1 0 0 0.4>)
     "Float (Alpha)"=&Multiply_Float(
      A=&FN_FrontEndFloatModifier(Gameplay=0.4, Front End=0.2)
      B=&Multiply_Float(A=&Lerp_Float(A=0, B=1, Alpha=$User.ParaGlideLeanAlpha), B=$Particles.AlphaAll)
     )
    )
    Create Color Attribute=true
    Create SpriteSize Attribute=true
    Module Output:ENiagara_AttributeSamplingApplyOutput=::NewEnumerator1
    SpriteSize=&Multiply_Vector2D_ByFloat(Float=&Multiply_Float(A=70, B=$Particles.SizeAll), Vector2D=<1 1>)
   ) {module=FN_Sprite_Attributes001}
   &ScaleColor(
    Scale Alpha=&Multiply_Float(A=&Lerp_Float(A=0, B=1, Alpha=$User.ParaGlideLeanAlpha), B=$Particles.AlphaAll)
    ScaleA=true
    ScaleRGB=true
   )
   &CurlNoiseForce(Bias Noise=false, Noise Frequency=10, Noise Strength=400, Pan Noise=false)
   &SolveForcesAndVelocity(Clamp Velocity=false, Limit Acceleration=false)
  ]
  rendererClasses [&NiagaraSpriteRendererProperties]
  renderers [
   {rendererClass=&NiagaraSpriteRendererProperties}
   {rendererClass=&NiagaraSpriteRendererProperties, rendererIndex=1}
   {rendererClass=&NiagaraSpriteRendererProperties, rendererIndex=2}
  ]
 }
 FireUp {
  emitterUpdateScript [
   &EmitterState(
    Inactive Response=::NewEnumerator1
    Life Cycle Mode=::NewEnumerator1
    Loop Behavior=::NewEnumerator1
    Loop Duration Mode=::NewEnumerator1
    UseLoopDelay=false
   )
   &SpawnBurst_Instantaneous(Spawn Count=2, Use Spawn Probability=false, UseLoopCountLimit=false)
  ]
  particleSpawnScript [
   &PartitionParticles()
   /SetVariables(Particles.SizeScle:NiagaraFloat=&Multiply_Float(A=3, B=$System.FEsizeMatch)) {isSetParameters=true}
   /SetVariables(Particles.VectorScale:Vector3f=<1 0.6 0.8>) {isSetParameters=true}
   &InitializeParticle(
    Color=<rgba 1 1 1 0.2>
    Mesh Scale=&Multiply_Vector(
     A=&Multiply_Vector(A=<1 1 1.5>, B=$Particles.VectorScale)
     B=&VectorFromFloat(Value=$Particles.SizeScle)
    )
    Mesh Scale Mode=::NewEnumerator4
    UsePositionOffset=false
    Write Material Random=true
   ) {module=InitializeParticle001}
   /SetVariables(Particles.Mesh_Color2:LinearColor=&MakeLinearColorFromVectorAndFloat(
     "Vector (RGB)"=&MakeVectorFromLinearColorRGB(Color=&MultiplyLinearColor(Linear Color=<rgba 1 0.3113 0.2093 1>, Scale Factor=&MakeVector4_FromFloat(Float=1.2)))
     "Float (Alpha)"=0.3
    )) {isSetParameters=true}
   /SetVariables(Particles.Mesh_Size2:Vector3f=&Multiply_VectorByFloat(Vector=&Multiply_Vector(A=<1 0.9 1.1>, B=$Particles.VectorScale), Float=$Particles.SizeScle)) {isSetParameters=true}
   &CameraOffset()
   &InitialMeshOrientation(
    Mesh Orientation Mode=::NewEnumerator2
    Rotation=&SelectVectorFromArray(
     Direct Array Index=$Particles.PartitionParticles.Partition
     Vector Selection Array=DataInterface() {propertyValues {FloatData [<0 0 0>, <0 0 0>]}}
    )
    Use Rotation Vector=true
   )
  ]
  particleUpdateScript [
   &ParticleState(Kill Particles When Lifetime Has Elapsed=false, Let Infinitely Lived Particles Die When Emitter Deactivates=true)
   &InterpolateOverTime(Rate Of Change=20, Target Vector Value:NiagaraPosition=$Particles.Position, Variable Type=::NewEnumerator2) {module=InterpolateOverTime001}
   /SetVariables(Particles.Interpolate Dynamic=&MakeFloat4FromVectorAndFloat("Vector (RGB)"=$Particles.InterpolateOverTime001.InterpolateOverTime_V3_.Moving Average, "Float (Alpha)"=0)) {isSetParameters=true}
   &SolveForcesAndVelocity(Clamp Velocity=false, Limit Acceleration=false)
   &ScaleColor(
    Scale Alpha=&Multiply_Float(A=1, B=0.5)
    Scale RGB=&Multiply_VectorByFloat(
     Vector=<1 1 1>
     Float=&MakeCustomFloatFromBool(Bool=$User.bIsFrontEndPreview, True Float=0.5, False Float=2)
    )
    ScaleA=true
    ScaleRGB=true
   )
   &DynamicMaterialParameters(
    Index 0 Param 2=1
    Index 0 Param 3=2
    Index 0 Param 4=1
    Index 3 Vector 1:NiagaraPosition=$System.SampleSkeletalMesh.SampledPosition
    Module.Index0Float0Write=true
    Module.Index0Float1Write=true
    Module.Index0Float2Write=true
    Module.Index0Float3Write=true
    Module.Index3Float3Write=true
    Module.Index3VectorWrite=true
    Write Parameter Index 3=::NewEnumerator3
   ) {module=DynamicMaterialParameters001}
   &SkeletalMeshLocation(
    Bone / Socket Index=$Particles.PartitionParticles.Partition
    Bone Sampling Mode=::NewEnumerator4
    MaskSpawnGroup=false
    Orientation Sampling=::NewEnumerator1
    Sampled Position Offset=&AddVectorToPosition(
     Position=&SelectVectorFromArray(
      Direct Array Index=$Particles.PartitionParticles.Partition
      Vector Selection Array=DataInterface() {propertyValues {FloatData [<40 0 -10>, <-40 0 -10>]}}
     )
     Vector=&Lerp_Vector(
      A=<0 0 0>
      B=<0 0 20>
      Alpha=&Waveform(
       "Waveform [1]"=::NewEnumerator0
       "[1] Phase Scale"=&SelectFloatFromArray(Direct Array Index=$Particles.PartitionParticles.Partition, Float Selection Array=DataInterface() {propertyValues {FloatData [0.5,0.7]}})
       "[1] Frequency"=&SelectFloatFromArray(Direct Array Index=$Particles.PartitionParticles.Partition, Float Selection Array=DataInterface() {propertyValues {FloatData [0.5,0.7]}})
       "[1] Amplitude Min/Max"=<0 1>
       "[1] Scale"=1.5
      )
     )
    )
    Sampled Position Offset Coordinate Space:ENiagaraSkelMeshTransforms=::NewEnumerator2
    Skeletal Mesh=$System.Skeletal Mesh Reader
   )
  ]
  rendererClasses [&NiagaraMeshRendererProperties]
  renderers [
   {rendererClass=&NiagaraMeshRendererProperties, rendererIndex=0}
   {rendererClass=&NiagaraMeshRendererProperties, rendererIndex=1}
  ]
 }
 Flame01 {
  emitterSpawnScript [/SetVariables(Emitter.WidthAll=1.7) {isSetParameters=true}]
  emitterUpdateScript [
   &EmitterState(
    Inactive Response=::NewEnumerator1
    Life Cycle Mode=::NewEnumerator1
    Loop Behavior=::NewEnumerator1
    Loop Duration Mode=::NewEnumerator1
    UseLoopDelay=false
   )
   &SpawnRate(Spawn Probability=&Lerp_Float(A=0, B=1, Alpha=$User.ParaGlideLeanAlpha), SpawnRate=6, Use Spawn Probability=true)
  ]
  particleSpawnScript [
   &PartitionParticles()
   &InitializeParticle(
    Color=<rgba 0.2708 0 0 1>
    Lifetime Max=1
    Lifetime Min=0.7
    Lifetime Mode=::NewEnumerator1
    Mesh Scale=&Multiply_VectorByFloat(Vector=<1 1 1.5>, Float=0.9)
    Mesh Scale Mode=::NewEnumerator4
    Position Offset=$System.FX position
    UsePositionOffset=true
    Write Material Random=true
   )
   /SetVariables(Particles.Dy3:Vector4f=&MakeVector4(
     X=&Multiply_Float(A=12, B=$Emitter.WidthAll)
     Y=&Multiply_Float(A=12, B=$Emitter.WidthAll)
     Z=0
     W=1
    )) {isSetParameters=true}
   /SetVariables(Particles.Dy4:Vector4f=&MakeVector4(
     X=&Multiply_Float(A=17, B=$Emitter.WidthAll)
     Y=&Multiply_Float(A=17, B=$Emitter.WidthAll)
     Z=0
     W=1
    )) {isSetParameters=true}
   &CameraOffset()
   &InitialMeshOrientation(Mesh Orientation Mode=::NewEnumerator2, Rotation=<0 0 0.4>, Use Rotation Vector=true)
   &InitialMeshOrientation(
    Mesh Orientation Mode=::NewEnumerator2
    Rotation=&SelectVectorFromArray(
     Direct Array Index=$Particles.PartitionParticles.Partition
     Vector Selection Array=DataInterface() {propertyValues {FloatData [<0 0 0.3>, <0 0 0.3>]}}
    )
    Use Rotation Vector=true
   ) {module=InitialMeshOrientation001}
   &InitialMeshOrientation(
    Mesh Orientation Mode=::NewEnumerator2
    Rotation=&RandomRangeVector(Minimum=<0 0 0>, Maximum=<0 0 1>)
    Use Rotation Vector=true
   ) {module=InitialMeshOrientation002}
   &AddVelocity(Velocity=<0 0 80>)
  ]
  particleUpdateScript [
   &DynamicMaterialParameters(
    Index 0 Param 1=&Multiply_Float(A=18, B=$Emitter.WidthAll)
    Index 0 Param 2=&Multiply_Float(A=18, B=$Emitter.WidthAll)
    Index 1 Param 4=&FloatFromCurve(FloatCurve=DataInterface() {propertyValues {Curve {defaultValue=1, keys [{tangentMode=RCTM_User, time=0, value=0}, {tangentMode=RCTM_User, time=1, value=0.8}]}}})
    Module.Index0Param0Write=true
    Module.Index0Param1Write=true
    Module.Index0Param2Write=true
    Module.Index0Param3Write=true
    Module.Index1Param0Write=true
    Module.Index1Param1Write=true
    Module.Index1Param2Write=true
    Module.Index1Param3Write=true
    Module.Index2Param0Write=true
    Module.Index2Param1Write=true
    Module.Index2Param2Write=true
    Module.Index2Param3Write=true
    Write Parameter Index 1=true
    Write Parameter Index 2=true
   )
   /SetVariables(Particles.ScaleAlphaAll:NiagaraFloat=&FloatFromCurve(FloatCurve=DataInterface() {propertyValues {Curve {
        defaultValue=1
        keys [
         {interpMode=RCIM_Cubic, tangentMode=RCTM_User, time=0, value=0}
         {arriveTangent=2, interpMode=RCIM_Cubic, leaveTangent=2, time=0.2, value=1}
         {interpMode=RCIM_Cubic, tangentMode=RCTM_User, time=0.5, value=1}
         {tangentMode=RCTM_User, time=1, value=0}
        ]
       }}})) {isSetParameters=true}
   &ParticleState()
   &ScaleMeshSize(Scale Factor=&Multiply_Vector(
     A=&Multiply_VectorByFloat(
      Vector=&SelectVectorFromArray(
       Direct Array Index=$Particles.PartitionParticles.Partition
       Vector Selection Array=DataInterface() {propertyValues {FloatData [<1 1 1>, <1 1 1>]}}
      )
      Float=0.31
     )
     B=<1 1 1>
    ))
   &ScaleColor(
    Scale Alpha=$Particles.ScaleAlphaAll
    Scale RGB=&Multiply_VectorByFloat(Vector=<1 1 1>, Float=1)
    ScaleA=true
    ScaleRGB=true
   )
   /SetVariables(Particles.Color3:LinearColor=&MultiplyLinearColor(Linear Color=<rgba 1.5 1.5 1.5 1>, Scale Factor=&MakeVector4_FromFloat(Float=$Particles.ScaleAlphaAll))) {isSetParameters=true}
   /SetVariables(Particles.Color4:LinearColor=&MultiplyLinearColor(
     Linear Color=<rgba 1 0.1406 0.2122 0.2>
     Scale Factor=&MakeVector4_FromFloat(Float=$Particles.ScaleAlphaAll)
    )) {isSetParameters=true}
   &ScaleMeshSize(Scale Factor=&Multiply_VectorByFloat(
     Vector=&RandomRangeVector(Minimum=<0.5 0.5 0.7>, Maximum=<1.3 1.3 1>)
     Float=&Multiply_Float(A=&RandomRangeFloat(Minimum=1, Maximum=1.1), B=$System.FEsizeMatch)
    )) {module=ScaleMeshSize001}
   &SkeletalMeshLocation(
    Bone / Socket Index=$Particles.PartitionParticles.Partition
    Bone Sampling Mode=::NewEnumerator4
    MaskSpawnGroup=false
    Orientation Sampling=::NewEnumerator1
    Position Sampling=::NewEnumerator1
    Sampled Position Offset=&AddVectorToPosition(
     Position=&SelectVectorFromArray(
      Direct Array Index=$Particles.PartitionParticles.Partition
      Vector Selection Array=DataInterface() {propertyValues {FloatData [<40 0 0>, <-40 0 0>]}}
     )
     Vector=&Lerp_Vector(
      A=<0 0 0>
      B=<0 0 20>
      Alpha=&Waveform(
       "Waveform [1]"=::NewEnumerator0
       "[1] Phase Scale"=&SelectFloatFromArray(Direct Array Index=$Particles.PartitionParticles.Partition, Float Selection Array=DataInterface() {propertyValues {FloatData [0.5,0.7]}})
       "[1] Frequency"=&SelectFloatFromArray(Direct Array Index=$Particles.PartitionParticles.Partition, Float Selection Array=DataInterface() {propertyValues {FloatData [0.5,0.7]}})
       "[1] Amplitude Min/Max"=<0 1>
       "[1] Scale"=1.5
      )
     )
    )
    Sampled Position Offset Coordinate Space:ENiagaraSkelMeshTransforms=::NewEnumerator2
    Skeletal Mesh=$System.Skeletal Mesh Reader
   )
   &SolveForcesAndVelocity(Clamp Velocity=false, Limit Acceleration=false)
  ]
  rendererClasses [&NiagaraMeshRendererProperties]
  renderers [
   {rendererClass=&NiagaraMeshRendererProperties, rendererIndex=0}
   {rendererClass=&NiagaraMeshRendererProperties, rendererIndex=1}
   {rendererClass=&NiagaraMeshRendererProperties, rendererIndex=2}
  ]
 }
 Head {
  emitterUpdateScript [
   &EmitterState(Life Cycle Mode=::NewEnumerator1, Loop Behavior=::NewEnumerator1, Loop Duration Mode=::NewEnumerator1, UseLoopDelay=false)
   &SpawnBurst_Instantaneous(Spawn Count=2, Use Spawn Probability=false, UseLoopCountLimit=false)
  ]
  particleSpawnScript [
   &PartitionParticles()
   &InitializeParticle(
    Color=<rgba 0.7 0.7 0.7 1>
    Mesh Scale=&Multiply_VectorByFloat(Vector=<1 1 1>, Float=0.7)
    Mesh Scale Mode=::NewEnumerator4
    UsePositionOffset=false
    Write Material Random=true
   )
   &InitialMeshOrientation(
    Mesh Orientation Mode=::NewEnumerator2
    Rotation=&SelectVectorFromArray(
     Direct Array Index=$Particles.PartitionParticles.Partition
     Vector Selection Array=DataInterface() {propertyValues {FloatData [<0 0 0.25>, <0 0 0.25>]}}
    )
    Use Rotation Vector=true
   )
   &InitialMeshOrientation(
    Mesh Orientation Mode=::NewEnumerator2
    Rotation=&SelectVectorFromArray(
     Direct Array Index=$Particles.PartitionParticles.Partition
     Vector Selection Array=DataInterface() {propertyValues {FloatData [<0 0.2 0>, <0 0.18 0>]}}
    )
    Use Rotation Vector=true
   ) {module=InitialMeshOrientation001}
   &InitialMeshOrientation(Mesh Orientation Mode=::NewEnumerator2, Rotation=<0 0 0.25>, Use Rotation Vector=true) {module=InitialMeshOrientation002}
  ]
  particleUpdateScript [
   &ParticleState(Kill Particles When Lifetime Has Elapsed=false)
   &ScaleMeshSize(Scale Factor=&Multiply_VectorByFloat(Vector=<1 1 1>, Float=&Multiply_Float(A=0.8, B=$System.FEsizeMatch)))
   &SkeletalMeshLocation(
    Bone / Socket Index=$Particles.PartitionParticles.Partition
    Bone Sampling Mode=::NewEnumerator4
    MaskSpawnGroup=false
    Orientation Sampling=::NewEnumerator1
    Sampled Position Offset=&AddVectorToPosition(
     Position=&AddVectorToPosition(
      Position=&SelectVectorFromArray(
       Direct Array Index=$Particles.PartitionParticles.Partition
       Vector Selection Array=DataInterface() {propertyValues {FloatData [<40 0 -10>, <-40 0 -10>]}}
      )
      Vector=&Lerp_Vector(
       A=<0 0 0>
       B=<0 0 20>
       Alpha=&Waveform(
        "Waveform [1]"=::NewEnumerator0
        "[1] Phase Scale"=&SelectFloatFromArray(Direct Array Index=$Particles.PartitionParticles.Partition, Float Selection Array=DataInterface() {propertyValues {FloatData [0.5,0.7]}})
        "[1] Frequency"=&SelectFloatFromArray(Direct Array Index=$Particles.PartitionParticles.Partition, Float Selection Array=DataInterface() {propertyValues {FloatData [0.5,0.7]}})
        "[1] Amplitude Min/Max"=<0 1>
        "[1] Scale"=1.5
       )
      )
     )
     Vector=<0 0 0>
    )
    Sampled Position Offset Coordinate Space:ENiagaraSkelMeshTransforms=::NewEnumerator2
    Skeletal Mesh=$System.Skeletal Mesh Reader
   )
   &Light_Attributes(
    Color=1
    Color Value To Scale:LinearColor=&MakeLinearColorFromVectorAndFloat("Vector (RGB)"=&MakeVectorFromLinearColorRGB(Color=<rgba 30 1 0 1>), "Float (Alpha)"=1)
    Enabled=1
    Falloff Exponent=1
    Light Exponent=3
    Light Position Offset:Vector3f=<0 0 40>
    Light Radius=150
    Position=1
    Position Offset Coordinate Space:ENiagaraCoordinateSpace=::Local
    Radius=1
    ScaleA=true
    ScaleRGB=true
    UsePositionOffset=true
   )
  ]
  rendererClasses [&NiagaraMeshRendererProperties, &NiagaraLightRendererProperties]
  renderers [
   {rendererClass=&NiagaraMeshRendererProperties, rendererIndex=0}
   {rendererClass=&NiagaraLightRendererProperties, rendererIndex=1}
  ]
 }
 Smoke {
  emitterUpdateScript [
   &EmitterState()
   &SpawnRate(
    Spawn Probability=&Lerp_Float(A=0.5, B=1, Alpha=$User.ParaGlideLeanAlpha)
    SpawnRate=40
    Use Spawn Probability=true
   )
  ]
  particleSpawnScript [
   &PartitionParticles()
   &InitializeParticle(
    Color=&MakeCustomLinearColorFromBool(Bool=&RandomBool(), True Color=<rgba 3 0.223 0.0768 2>, False Color=<rgba 0.1 0 0 5>)
    Lifetime Min=1.5
    Lifetime Mode=::NewEnumerator1
    Sprite Rotation Mode=::NewEnumerator2
    Sprite Size Mode=::NewEnumerator1
    Uniform Sprite Size Max=150
    Uniform Sprite Size Min=100
    UsePositionOffset=false
    Write Material Random=true
   )
   &CameraOffset(Camera Offset Amount=-5)
   &ShapeLocation(
    MaskSpawnGroup=false
    Override Seed=false
    Shape Origin=&MakeVector(X=0, Y=0, Z=0)
    Sphere Radius=3
    UseOwnerScale=true
   )
   &AddVelocity(
    Cone Angle=10
    Cone Axis=<0 0 1>
    Module.UseConeFalloff=false
    Module.UseVelDistribution=false
    Override Seed=false
    Velocity Mode=::NewEnumerator2
    Velocity Speed=200
   )
  ]
  particleUpdateScript [
   &SolveForcesAndVelocity(Clamp Velocity=false, Limit Acceleration=false)
   &ParticleState()
   &ScaleSpriteSize(Uniform Curve Sprite Scale=DataInterface() {propertyValues {Curve {
       defaultValue=0
       keys [
        {interpMode=RCIM_Cubic, tangentMode=RCTM_User, time=0, value=0.3}
        {arriveTangent=-0.1, interpMode=RCIM_Cubic, leaveTangent=-0.1, time=0.2941, value=1.2078}
        {interpMode=RCIM_Cubic, tangentMode=RCTM_User, time=1, value=0.2}
       ]
      }}})
   &DynamicMaterialParameters(
    Index 0 Param 1=&MakeCustomFloatFromBool(Bool=&RandomBool(), True Float=0, False Float=0.75)
    Module.Param0WriteEnabled=true
    Module.Param1WriteEnabled=true
    Module.Param2WriteEnabled=true
    Module.Param3WriteEnabled=true
   )
   &ScaleColor(
    Scale Alpha=&FloatFromCurve(
     FloatCurve=DataInterface() {propertyValues {Curve {keys [
         {interpMode=RCIM_Cubic, tangentMode=RCTM_User, time=0, value=0}
         {interpMode=RCIM_Cubic, tangentMode=RCTM_User, time=0.5, value=1}
         {tangentMode=RCTM_User, time=1, value=0}
        ]}}}
     Scale Curve=0.3
    )
    ScaleA=true
    ScaleRGB=true
   )
   &SpriteRotationRate(Rotation Rate=&RandomRangeFloat(Minimum=-70, Maximum=70))
   &ScaleColor(
    Scale RGB=&MakeVectorFromLinearColorRGB(Color=&ColorFromCurve(ColorCurve=DataInterface() {propertyValues {
        AlphaCurve {keys []}
        BlueCurve {keys [<at 0 0.0021 RCIM_Cubic>,<at 0.345 0 RCIM_Cubic>]}
        GreenCurve {keys [<at 0 3.5705 RCIM_Cubic>,<at 0.345 0 RCIM_Cubic>]}
        RedCurve {keys [<at 0 5 RCIM_Cubic>,<at 0.345 1 RCIM_Cubic>]}
       }}))
    ScaleA=true
    ScaleRGB=true
   ) {module=ScaleColor001}
   &SkeletalMeshLocation(
    Bone / Socket Index=$Particles.PartitionParticles.Partition
    Bone Sampling Mode=::NewEnumerator4
    MaskSpawnGroup=false
    Orientation Sampling=::NewEnumerator1
    Position Sampling=::NewEnumerator1
    Sampled Position Offset=&AddVectorToPosition(
     Position=&SelectVectorFromArray(
      Direct Array Index=$Particles.PartitionParticles.Partition
      Vector Selection Array=DataInterface() {propertyValues {FloatData [<40 0 0>, <-40 0 0>]}}
     )
     Vector=&Lerp_Vector(
      A=<0 0 0>
      B=<0 0 20>
      Alpha=&Waveform(
       "Waveform [1]"=::NewEnumerator0
       "[1] Phase Scale"=&SelectFloatFromArray(Direct Array Index=$Particles.PartitionParticles.Partition, Float Selection Array=DataInterface() {propertyValues {FloatData [0.5,0.7]}})
       "[1] Frequency"=&SelectFloatFromArray(Direct Array Index=$Particles.PartitionParticles.Partition, Float Selection Array=DataInterface() {propertyValues {FloatData [0.5,0.7]}})
       "[1] Amplitude Min/Max"=<0 1>
       "[1] Scale"=1.5
      )
     )
    )
    Sampled Position Offset Coordinate Space:ENiagaraSkelMeshTransforms=::NewEnumerator2
    Skeletal Mesh=$System.Skeletal Mesh Reader
   )
  ]
  rendererClasses [&NiagaraSpriteRendererProperties]
  renderers [{rendererClass=&NiagaraSpriteRendererProperties}]
 }
 Sparks {
  emitterUpdateScript [
   &EmitterState(Life Cycle Mode=::NewEnumerator1, Loop Duration=1, UseLoopDelay=false)
   &SpawnRate(
    Spawn Probability=&Lerp_Float(A=0.8, B=1, Alpha=$User.ParaGlideLeanAlpha)
    SpawnRate=120
    Use Spawn Probability=true
   )
   &SpawnBurst_Instantaneous(Spawn Count=10, Use Spawn Probability=false, UseLoopCountLimit=false)
  ]
  particleSpawnScript [
   &PartitionParticles()
   &InitializeParticle(
    Color=<rgba 10 1.3868 1.3868 1>
    Lifetime Min=2
    Lifetime Mode=::NewEnumerator1
    Sprite Size Max=<1 10>
    Sprite Size Min=<1 5>
    Sprite Size Mode=::NewEnumerator2
    UsePositionOffset=false
    Write Material Random=true
   )
   &SkeletalMeshLocation(
    Bone / Socket Index=$Particles.PartitionParticles.Partition
    Bone Sampling Mode=::NewEnumerator4
    MaskSpawnGroup=false
    Orientation Sampling=::NewEnumerator1
    Sampled Position Offset=&AddVectorToPosition(
     Position=&SelectVectorFromArray(
      Direct Array Index=$Particles.PartitionParticles.Partition
      Vector Selection Array=DataInterface() {propertyValues {FloatData [<40 0 0>, <-40 0 0>]}}
     )
     Vector=<0 0 20>
    )
    Sampled Position Offset Coordinate Space:ENiagaraSkelMeshTransforms=::NewEnumerator2
    Skeletal Mesh=$System.Skeletal Mesh Reader
   )
   &ShapeLocation(Handle Radius=2, Large Radius=2, MaskSpawnGroup=false, Override Seed=false, Shape Primitive=::NewEnumerator3, UseOwnerScale=true)
   &AddVelocity(
    Velocity=&MakeVector(
     X=&Sine(Normalized Angle=$Emitter.Age, Period=0.7, Scale=50, Bias=0)
     Y=&Cosine(Normalized Angle=$Emitter.Age, Period=0.7, Scale=50, Bias=0)
     Z=50
    )
    Velocity Speed Scale=1.3
   )
  ]
  particleUpdateScript [
   &ParticleState()
   /Niagara/Modules/Update/Forces/LineAttractionForce(
    Attraction Strength=300
    Line End:Vector3f=&AddVectorToPosition(Position=<0 0 1000>, Vector=Linked() {linkedVariable {name=Particles.SkeletalMeshLocation.SampledPosition, type=NiagaraPosition}})
    Line Start:Vector3f=&AddVectorToPosition(Position=<0 0 1000>, Vector=Linked() {linkedVariable {name=Particles.SkeletalMeshLocation.SampledPosition, type=NiagaraPosition}})
    Use Attraction Falloff=false
   )
   /SetVariables(Particles.Glow03_Color=&MakeLinearColorFromVectorAndFloat(
     "Vector (RGB)"=&MakeVectorFromLinearColorRGB(Color=&MultiplyLinearColor(Linear Color=<rgba 1 0.0112 0.0112 1>, Scale Factor=&MakeVector4_FromFloat(Float=1)))
     "Float (Alpha)"=&Multiply_Float(
      A=&Multiply_Float(A=0.2, B=1)
      B=&FN_FrontEndFloatModifier(Gameplay=1, Front End=0.7)
     )
    )) {isSetParameters=true}
   /SetVariables(Particles.Glow03_Size=&Multiply_Vector2D_ByFloat(
     Float=&FloatFromCurve(
      FloatCurve=DataInterface() {propertyValues {Curve {keys [
          <at 0 0 RCIM_Cubic>
          <at 0.2 1.3 RCIM_Cubic>
          {arriveTangent=-0.1423, interpMode=RCIM_Cubic, leaveTangent=-0.1423, tangentMode=RCTM_User, time=1, value=0}
         ]}}}
      Scale Curve=15
     )
     Vector2D=<0.7 1.5>
    )) {isSetParameters=true}
   &ScaleColor(
    Scale Alpha=&FloatFromCurve(FloatCurve=DataInterface() {propertyValues {Curve {keys [<at 0.3 1>,<at 0.4 1>,<at 1 1>]}}})
    ScaleA=true
    ScaleRGB=true
   )
   &ScaleSpriteSize(Scale Factor=&Vector2DFromFloat(Value=&FloatFromCurve(
      FloatCurve=DataInterface() {propertyValues {Curve {keys [
          <at 0 0 RCIM_Cubic>
          {interpMode=RCIM_Cubic, leaveTangent=-3.0296, tangentMode=RCTM_Break, time=0.2, value=1.2}
          {arriveTangent=-1.8107, interpMode=RCIM_Cubic, leaveTangent=-1.8107, tangentMode=RCTM_User, time=1, value=0}
         ]}}}
      Scale Curve=0.4
     ))) {module=ScaleSpriteSize001}
   &Drag(Drag=1.5, Use Linear Drag=true, Use Rotational Drag=false)
   &SolveForcesAndVelocity(Clamp Velocity=false, Limit Acceleration=false)
  ]
  rendererClasses [&NiagaraSpriteRendererProperties]
  renderers [
   {rendererClass=&NiagaraSpriteRendererProperties}
   {rendererClass=&NiagaraSpriteRendererProperties, rendererIndex=1}
  ]
 }
}
summary {
 systemName=NS_Contrail_MasakoReign
 emitters [
  {emitterName=Head, rendererClasses [&NiagaraMeshRendererProperties, &NiagaraLightRendererProperties]}
  {emitterName=Dragon, rendererClasses [&NiagaraRibbonRendererProperties]}
  {emitterName=FireUp, rendererClasses [&NiagaraMeshRendererProperties]}
  {emitterName=FireCylinder, rendererClasses [&NiagaraMeshRendererProperties]}
  {emitterName=Flame01, rendererClasses [&NiagaraMeshRendererProperties]}
  {emitterName=Smoke, rendererClasses [&NiagaraSpriteRendererProperties]}
  {emitterName=Sparks, rendererClasses [&NiagaraSpriteRendererProperties]}
  {emitterName=CurveFire, rendererClasses [&NiagaraMeshRendererProperties]}
  {emitterName=FireSprite, rendererClasses [&NiagaraSpriteRendererProperties]}
 ]
 userVariables [
  User.bIsFrontEnd {type=NiagaraBool}
  User.bIsFrontEndPreview {type=NiagaraBool, defaultValue=true}
  User.ParaGlideLeanAlpha {type=NiagaraFloat, defaultValue=1, description=""}
 ]
}
